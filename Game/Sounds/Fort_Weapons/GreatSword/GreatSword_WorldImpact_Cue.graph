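# Sound cue for greatsword impacts against the world, using pitched-up Revenant world impact waves.
# A quality-level node selects between two filled branches: the first randomizes seven waves, the second an evenly weighted subset of four.
# Each branch passes through a modulator that randomly raises pitch.
# The cue uses small melee attenuation, the melee weapon sound class and player melee weapon concurrency, with slightly boosted volume.
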SoundCue /Game/Sounds/Fort_Weapons/GreatSword/GreatSword_WorldImpact_Cue
root=QualityLevel_0
nodes [
 {id=Modulator_1, children [Random_1], properties {pitchMax=1.5, pitchMin=1.3}}
 {id=Modulator_2, children [Random_2], properties {pitchMax=1.5, pitchMin=1.3}}
 {id=QualityLevel_0, children [Modulator_1,Modulator_2,null,null]}
 {
  id=Random_1
  children [WavePlayer_7, WavePlayer_8, WavePlayer_9, WavePlayer_10, WavePlayer_11, WavePlayer_12, WavePlayer_13]
  properties {weights [1,1,1,1,1,1,1]}
 }
 {id=Random_2, children [WavePlayer_14, WavePlayer_15, WavePlayer_18, WavePlayer_20], properties {weights [1,1,1,1]}}
 {id=WavePlayer_10, properties {soundWaveAssetPtr=&/Revenant_WorldImpact_004}}
 {id=WavePlayer_11, properties {soundWaveAssetPtr=&/Revenant_WorldImpact_005}}
 {id=WavePlayer_12, properties {soundWaveAssetPtr=&/Revenant_WorldImpact_006}}
 {id=WavePlayer_13, properties {soundWaveAssetPtr=&/Revenant_WorldImpact_007}}
 {id=WavePlayer_14, properties {soundWaveAssetPtr=&/Revenant_WorldImpact_001}}
 {id=WavePlayer_15, properties {soundWaveAssetPtr=&/Revenant_WorldImpact_002}}
 {id=WavePlayer_18, properties {soundWaveAssetPtr=&Revenant_WorldImpact_005}}
 {id=WavePlayer_20, properties {soundWaveAssetPtr=&Revenant_WorldImpact_007}}
 {id=WavePlayer_7, properties {soundWaveAssetPtr=&Revenant_WorldImpact_001}}
 {id=WavePlayer_8, properties {soundWaveAssetPtr=&Revenant_WorldImpact_002}}
 {id=WavePlayer_9, properties {soundWaveAssetPtr=&/Revenant_WorldImpact_003}}
]
properties {
 attenuationSettings=&Fort_Weapon_Melee_Sml_Att
 maxDistance=2800
 soundClassObject=&Fort_WeaponMelee_PSM
 volumeMultiplier=1.2
 concurrencySet [&Player_Melee_Weapon_Conc]
}
